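# Animation Blueprint for the TangyRadishMagma backpack cosmetic, posing its wings to match player movement and adding secondary motion.
# The wingPosition state machine, with a nested skydiving machine, selects pose-asset poses, most layered with additive lean blendspaces.
# A pack bone offset and chained AnimDynamics nodes follow, simulating leaf bones fully and leg bones at low alpha.
# The event graph updates lean vectors and sets petal simulation parameters around a front-end check.
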AnimBlueprint /BRCosmetics/Accessories/FORT_Backpacks/Backpack_F_MED_TangyRadishMagma/Meshes/F_MED_TangyRadishMagma_Pack_AnimBP
target_skeleton=&/F_MED_TangyRadishMagma_Pack_Skeleton
graphs [
 AnimGraph {nodes [
   Input Pose()
   AnimGraphNode_StateMachine_0 = wingPosition() {is=title AnimGraphNode_StateMachine}
   n1 = "Transform (Modify) Bone - Bone: pack"(ComponentPose=Local To Component(LocalPose=@AnimGraphNode_StateMachine_0), Translation="0,0.4,0", Rotation="0,0,0") {
    rotationMode=BMM_Additive
    rotationSpace=BCS_BoneSpace
    translationMode=BMM_Additive
    translationSpace=BCS_BoneSpace
    alphaBoolBlend {blendOption=Linear}
    boneToModify {boneName=pack}
    translation {y=0.4}
   }
   n2 = Anim Dynamics - Bone: dyn_leaf_5_00!(SimulationSpace:AnimPhysSimSpaceType Enum, ComponentPose=@n1, Alpha=1, GravityOverride:Vector="0,0,0", ExternalForce="0,0,0") -> (Pose) {
    angularDampingOverride=1
    angularSpringConstant=1000
    bAngularSpring=true
    bOverrideAngularDamping=true
    bUseGravityOverride=true
    gravityScale=0.35
    simulationSpace=RootRelative
    alphaBoolBlend {blendOption=Linear}
    boundBone {boneName=dyn_leaf_5_00}
    componentLinearAccScale=<1 1 1>
    componentLinearVelScale=<1 1 1>
    gravityOverride {z=-5000}
    physicsBodyDefinitions [{
      boundBone {boneName=dyn_leaf_5_00}
      boxExtents=<20 20 20>
      constraintSetup {angularLimitsMax=<0 2 0>, angularLimitsMin=<0 -2 -45>, angularTarget=<1 0 0>}
      localJointOffset=<1 0 0>
     }]
    simSpaceSettings {maxAngularAcceleration=35, maxAngularVelocity=7, simSpaceAngularAlpha=1}
   }
   n3 = Anim Dynamics - Bone: dyn_leaf_4_00!(SimulationSpace:AnimPhysSimSpaceType Enum, ComponentPose=@n2, Alpha=1, GravityOverride:Vector="0,0,-5000", ExternalForce="0,0,0") -> (Pose) {
    angularDampingOverride=1
    angularSpringConstant=1000
    bAngularSpring=true
    bOverrideAngularDamping=true
    bUseGravityOverride=true
    gravityScale=0.35
    simulationSpace=RootRelative
    alphaBoolBlend {blendOption=Linear}
    boundBone {boneName=dyn_leaf_4_00}
    componentLinearAccScale=<1 1 1>
    componentLinearVelScale=<1 1 1>
    gravityOverride {z=-5000}
    physicsBodyDefinitions [{
      boundBone {boneName=dyn_leaf_4_00}
      boxExtents=<25 25 25>
      constraintSetup {angularLimitsMax=<0 1 0>, angularLimitsMin=<0 -1 -45>, angularTarget=<1 0 0>}
      localJointOffset=<1 0 0>
     }]
    simSpaceSettings {maxAngularAcceleration=35, maxAngularVelocity=7, simSpaceAngularAlpha=1}
   }
   n4 = Anim Dynamics - Bone: dyn_leaf_4_01!(SimulationSpace:AnimPhysSimSpaceType Enum, ComponentPose=@n3, Alpha=1, GravityOverride:Vector="0,0,-5000", ExternalForce="0,0,0") -> (Pose) {
    angularDampingOverride=1
    angularSpringConstant=1000
    bAngularSpring=true
    bOverrideAngularDamping=true
    bUseGravityOverride=true
    gravityScale=0.35
    simulationSpace=RootRelative
    alphaBoolBlend {blendOption=Linear}
    boundBone {boneName=dyn_leaf_4_01}
    componentLinearAccScale=<1 1 1>
    componentLinearVelScale=<1 1 1>
    gravityOverride {z=-5000}
    physicsBodyDefinitions [{
      boundBone {boneName=dyn_leaf_4_01}
      boxExtents=<25 25 25>
      constraintSetup {angularLimitsMax=<0 1 0>, angularLimitsMin=<0 -1 -55>, angularTarget=<1 0 0>}
      localJointOffset=<1 0 0>
     }]
    simSpaceSettings {maxAngularAcceleration=35, maxAngularVelocity=7, simSpaceAngularAlpha=1}
   }
   n5 = Anim Dynamics - Bone: dyn_leaf_3_01!(SimulationSpace:AnimPhysSimSpaceType Enum, ComponentPose=@n4, Alpha=1, GravityOverride:Vector="0,0,-5000", ExternalForce="0,0,0") -> (Pose) {
    angularDampingOverride=1
    angularSpringConstant=1000
    bAngularSpring=true
    bOverrideAngularDamping=true
    bUseGravityOverride=true
    gravityScale=0.35
    simulationSpace=RootRelative
    alphaBoolBlend {blendOption=Linear}
    boundBone {boneName=dyn_leaf_3_01}
    componentLinearAccScale=<1 1 1>
    componentLinearVelScale=<1 1 1>
    gravityOverride {z=-5000}
    physicsBodyDefinitions [{
      boundBone {boneName=dyn_leaf_3_01}
      boxExtents=<30 30 30>
      constraintSetup {angularLimitsMax=<0 2 25>, angularLimitsMin=<0 -2 -7>, angularTarget=<1 0 0>}
      localJointOffset=<1 0 0>
     }]
    simSpaceSettings {maxAngularAcceleration=35, maxAngularVelocity=7, simSpaceAngularAlpha=1}
   }
   n6 = Anim Dynamics - Bone: dyn_leaf_3_02!(SimulationSpace:AnimPhysSimSpaceType Enum, ComponentPose=@n5, Alpha=1, GravityOverride:Vector="0,0,-5000", ExternalForce="0,0,0") -> (Pose) {
    angularDampingOverride=1
    angularSpringConstant=1000
    bAngularSpring=true
    bOverrideAngularDamping=true
    bUseGravityOverride=true
    gravityScale=0.35
    simulationSpace=RootRelative
    alphaBoolBlend {blendOption=Linear}
    boundBone {boneName=dyn_leaf_3_02}
    componentLinearAccScale=<1 1 1>
    componentLinearVelScale=<1 1 1>
    gravityOverride {z=-5000}
    physicsBodyDefinitions [{
      boundBone {boneName=dyn_leaf_3_02}
      boxExtents=<30 30 30>
      constraintSetup {angularLimitsMax=<0 1 35>, angularLimitsMin=<0 -1 0>, angularTarget=<1 0 0>}
      localJointOffset=<1 0 0>
     }]
    simSpaceSettings {maxAngularAcceleration=35, maxAngularVelocity=7, simSpaceAngularAlpha=1}
   }
   n7 = Anim Dynamics - Bone: dyn_leaf_2_01!(SimulationSpace:AnimPhysSimSpaceType Enum, ComponentPose=@n6, Alpha=1, GravityOverride:Vector="0,0,-5000", ExternalForce="0,0,0") -> (Pose) {
    angularDampingOverride=1
    angularSpringConstant=1000
    bAngularSpring=true
    bOverrideAngularDamping=true
    bUseGravityOverride=true
    gravityScale=0.35
    simulationSpace=RootRelative
    alphaBoolBlend {blendOption=Linear}
    boundBone {boneName=dyn_leaf_2_01}
    componentLinearAccScale=<1 1 1>
    componentLinearVelScale=<1 1 1>
    gravityOverride {z=-5000}
    physicsBodyDefinitions [{
      boundBone {boneName=dyn_leaf_2_01}
      boxExtents=<52 52 52>
      constraintSetup {angularLimitsMax=<0 2 0>, angularLimitsMin=<0 -2 -25>, angularTarget=<1 0 0>}
      localJointOffset=<1 0 0>
     }]
    simSpaceSettings {maxAngularAcceleration=35, maxAngularVelocity=7, simSpaceAngularAlpha=1}
   }
   n8 = Anim Dynamics - Bone: dyn_leaf_2_02!(SimulationSpace:AnimPhysSimSpaceType Enum, ComponentPose=@n7, Alpha=1, GravityOverride:Vector="0,0,-5000", ExternalForce="0,0,0") -> (Pose) {
    angularDampingOverride=1
    angularSpringConstant=1000
    bAngularSpring=true
    bOverrideAngularDamping=true
    bUseGravityOverride=true
    gravityScale=0.35
    simulationSpace=RootRelative
    alphaBoolBlend {blendOption=Linear}
    boundBone {boneName=dyn_leaf_2_02}
    componentLinearAccScale=<1 1 1>
    componentLinearVelScale=<1 1 1>
    gravityOverride {z=-5000}
    physicsBodyDefinitions [{
      boundBone {boneName=dyn_leaf_2_02}
      boxExtents=<32 32 32>
      constraintSetup {angularLimitsMax=<0 1 0>, angularLimitsMin=<0 -1 -35>, angularTarget=<1 0 0>}
      localJointOffset=<1 0 0>
     }]
    simSpaceSettings {maxAngularAcceleration=35, maxAngularVelocity=7, simSpaceAngularAlpha=1}
   }
   n9 = Anim Dynamics - Bone: dyn_leaf_1_01!(SimulationSpace:AnimPhysSimSpaceType Enum, ComponentPose=@n8, Alpha=1, GravityOverride:Vector="0,0,-5000", ExternalForce="0,0,0") -> (Pose) {
    angularDampingOverride=1
    angularSpringConstant=1000
    bAngularSpring=true
    bOverrideAngularDamping=true
    bUseGravityOverride=true
    gravityScale=0.35
    simulationSpace=RootRelative
    alphaBoolBlend {blendOption=Linear}
    boundBone {boneName=dyn_leaf_1_01}
    componentLinearAccScale=<1 1 1>
    componentLinearVelScale=<1 1 1>
    gravityOverride {z=-5000}
    physicsBodyDefinitions [{
      boundBone {boneName=dyn_leaf_1_01}
      boxExtents=<38 38 38>
      constraintSetup {angularLimitsMax=<0 2 40>, angularLimitsMin=<0 -2 0>, angularTarget=<1 0 0>}
      localJointOffset=<1 0 0>
     }]
    simSpaceSettings {maxAngularAcceleration=35, maxAngularVelocity=7, simSpaceAngularAlpha=1}
   }
   n10 = Anim Dynamics - Bone: dyn_leaf_1_02!(SimulationSpace:AnimPhysSimSpaceType Enum, ComponentPose=@n9, Alpha=1, GravityOverride:Vector="0,0,-5000", ExternalForce="0,0,0") -> (Pose) {
    angularDampingOverride=1
    angularSpringConstant=1000
    bAngularSpring=true
    bOverrideAngularDamping=true
    bUseGravityOverride=true
    gravityScale=0.35
    simulationSpace=RootRelative
    alphaBoolBlend {blendOption=Linear}
    boundBone {boneName=dyn_leaf_1_02}
    componentLinearAccScale=<1 1 1>
    componentLinearVelScale=<1 1 1>
    gravityOverride {z=-5000}
    physicsBodyDefinitions [{
      boundBone {boneName=dyn_leaf_1_02}
      boxExtents=<28 28 28>
      constraintSetup {angularLimitsMax=<0 1 50>, angularLimitsMin=<0 -1 0>, angularTarget=<1 0 0>}
      localJointOffset=<1 0 0>
     }]
    simSpaceSettings {maxAngularAcceleration=35, maxAngularVelocity=7, simSpaceAngularAlpha=1}
   }
   n11 = Anim Dynamics - Bone: Leg_1_02_L(ComponentPose=@n10, Alpha=0.2, ExternalForce="0,0,0") {
    angularDampingOverride=1
    angularSpringConstant=10000
    bAngularSpring=true
    bOverrideAngularDamping=true
    bUseGravityOverride=true
    gravityScale=0.35
    simulationSpace=RootRelative
    alphaBoolBlend {blendOption=Linear}
    boundBone {boneName=Leg_1_02_L}
    componentLinearAccScale=<1 1 1>
    componentLinearVelScale=<1 1 1>
    physicsBodyDefinitions [{
      boundBone {boneName=Leg_1_02_L}
      boxExtents=<25 25 25>
      constraintSetup {angularLimitsMax=<0 0 50>, angularLimitsMin=<0 0 -15>, angularTarget=<1 0 0>}
      localJointOffset=<-1 0 0>
     }]
    simSpaceSettings {maxAngularAcceleration=35, maxAngularVelocity=7, simSpaceAngularAlpha=1}
   }
   n12 = Anim Dynamics - Bone: Leg_1_02_R(ComponentPose=@n11, Alpha=0.2, ExternalForce="0,0,0") {
    angularDampingOverride=1
    angularSpringConstant=10000
    bAngularSpring=true
    bOverrideAngularDamping=true
    bUseGravityOverride=true
    gravityScale=0.35
    simulationSpace=RootRelative
    alphaBoolBlend {blendOption=Linear}
    boundBone {boneName=Leg_1_02_R}
    componentLinearAccScale=<1 1 1>
    componentLinearVelScale=<1 1 1>
    physicsBodyDefinitions [{
      boundBone {boneName=Leg_1_02_R}
      boxExtents=<25 25 25>
      constraintSetup {angularLimitsMax=<0 0 50>, angularLimitsMin=<0 0 -15>, angularTarget=<1 0 0>}
      localJointOffset=<1 0 0>
     }]
    simSpaceSettings {maxAngularAcceleration=35, maxAngularVelocity=7, simSpaceAngularAlpha=1}
   }
   n13 = Anim Dynamics - Bone: Leg_2_02_L(ComponentPose=@n12, Alpha=0.2, ExternalForce="0,0,0") {
    angularDampingOverride=1
    angularSpringConstant=10000
    bAngularSpring=true
    bOverrideAngularDamping=true
    bUseGravityOverride=true
    gravityScale=0.35
    simulationSpace=RootRelative
    alphaBoolBlend {blendOption=Linear}
    boundBone {boneName=Leg_2_02_L}
    componentLinearAccScale=<1 1 1>
    componentLinearVelScale=<1 1 1>
    physicsBodyDefinitions [{
      boundBone {boneName=Leg_2_02_L}
      boxExtents=<25 25 25>
      constraintSetup {angularLimitsMax=<0 0 50>, angularLimitsMin=<0 0 -15>, angularTarget=<1 0 0>}
      localJointOffset=<-1 0 0>
     }]
    simSpaceSettings {maxAngularAcceleration=35, maxAngularVelocity=7, simSpaceAngularAlpha=1}
   }
   n14 = Anim Dynamics - Bone: Leg_2_02_R(ComponentPose=@n13, Alpha=0.2, ExternalForce="0,0,0") {
    angularDampingOverride=1
    angularSpringConstant=10000
    bAngularSpring=true
    bOverrideAngularDamping=true
    bUseGravityOverride=true
    gravityScale=0.35
    simulationSpace=RootRelative
    alphaBoolBlend {blendOption=Linear}
    boundBone {boneName=Leg_2_02_R}
    componentLinearAccScale=<1 1 1>
    componentLinearVelScale=<1 1 1>
    physicsBodyDefinitions [{
      boundBone {boneName=Leg_2_02_R}
      boxExtents=<25 25 25>
      constraintSetup {angularLimitsMax=<0 0 50>, angularLimitsMin=<0 0 -15>, angularTarget=<1 0 0>}
      localJointOffset=<1 0 0>
     }]
    simSpaceSettings {maxAngularAcceleration=35, maxAngularVelocity=7, simSpaceAngularAlpha=1}
   }
   n15 = Anim Dynamics - Bone: Leg_3_02_R(ComponentPose=@n14, Alpha=0.2, ExternalForce="0,0,0") {
    angularDampingOverride=1
    angularSpringConstant=10000
    bAngularSpring=true
    bOverrideAngularDamping=true
    bUseGravityOverride=true
    gravityScale=0.35
    simulationSpace=RootRelative
    alphaBoolBlend {blendOption=Linear}
    boundBone {boneName=Leg_3_02_R}
    componentLinearAccScale=<1 1 1>
    componentLinearVelScale=<1 1 1>
    physicsBodyDefinitions [{
      boundBone {boneName=Leg_3_02_R}
      boxExtents=<25 25 25>
      constraintSetup {angularLimitsMax=<0 0 50>, angularLimitsMin=<0 0 -15>, angularTarget=<1 0 0>}
      localJointOffset=<1 0 0>
     }]
    simSpaceSettings {maxAngularAcceleration=35, maxAngularVelocity=7, simSpaceAngularAlpha=1}
   }
   n16 = Anim Dynamics - Bone: Leg_4_02_R(ComponentPose=@n15, Alpha=0.2, ExternalForce="0,0,0") {
    angularDampingOverride=1
    angularSpringConstant=10000
    bAngularSpring=true
    bOverrideAngularDamping=true
    bUseGravityOverride=true
    gravityScale=0.35
    simulationSpace=RootRelative
    alphaBoolBlend {blendOption=Linear}
    boundBone {boneName=Leg_4_02_R}
    componentLinearAccScale=<1 1 1>
    componentLinearVelScale=<1 1 1>
    physicsBodyDefinitions [{
      boundBone {boneName=Leg_4_02_R}
      boxExtents=<25 25 25>
      constraintSetup {angularLimitsMax=<0 0 50>, angularLimitsMin=<0 0 -15>, angularTarget=<1 0 0>}
      localJointOffset=<1 0 0>
     }]
    simSpaceSettings {maxAngularAcceleration=35, maxAngularVelocity=7, simSpaceAngularAlpha=1}
   }
   n17 = Anim Dynamics - Bone: Leg_4_02_L(ComponentPose=@n16, Alpha=0.2, ExternalForce="0,0,0") {
    angularDampingOverride=1
    angularSpringConstant=10000
    bAngularSpring=true
    bOverrideAngularDamping=true
    bUseGravityOverride=true
    gravityScale=0.35
    simulationSpace=RootRelative
    alphaBoolBlend {blendOption=Linear}
    boundBone {boneName=Leg_4_02_L}
    componentLinearAccScale=<1 1 1>
    componentLinearVelScale=<1 1 1>
    physicsBodyDefinitions [{
      boundBone {boneName=Leg_4_02_L}
      boxExtents=<25 25 25>
      constraintSetup {angularLimitsMax=<0 0 50>, angularLimitsMin=<0 0 -15>, angularTarget=<1 0 0>}
      localJointOffset=<-1 0 0>
     }]
    simSpaceSettings {maxAngularAcceleration=35, maxAngularVelocity=7, simSpaceAngularAlpha=1}
   }
   n18 = Anim Dynamics - Bone: Leg_3_02_L(ComponentPose=@n17, Alpha=0.2, ExternalForce="0,0,0") {
    angularDampingOverride=1
    angularSpringConstant=10000
    bAngularSpring=true
    bOverrideAngularDamping=true
    bUseGravityOverride=true
    gravityScale=0.35
    simulationSpace=RootRelative
    alphaBoolBlend {blendOption=Linear}
    boundBone {boneName=Leg_3_02_L}
    componentLinearAccScale=<1 1 1>
    componentLinearVelScale=<1 1 1>
    physicsBodyDefinitions [{
      boundBone {boneName=Leg_3_02_L}
      boxExtents=<25 25 25>
      constraintSetup {angularLimitsMax=<0 0 50>, angularLimitsMin=<0 0 -15>, angularTarget=<1 0 0>}
      localJointOffset=<-1 0 0>
     }]
    simSpaceSettings {maxAngularAcceleration=35, maxAngularVelocity=7, simSpaceAngularAlpha=1}
   }
   n19 = Anim Dynamics - Bone: Leg_3_01_L(ComponentPose=@n18, Alpha=0.2, ExternalForce="0,0,0") {
    angularDampingOverride=1
    angularSpringConstant=10000
    bAngularSpring=true
    bOverrideAngularDamping=true
    bUseGravityOverride=true
    gravityScale=0.35
    simulationSpace=RootRelative
    alphaBoolBlend {blendOption=Linear}
    boundBone {boneName=Leg_3_01_L}
    componentLinearAccScale=<1 1 1>
    componentLinearVelScale=<1 1 1>
    physicsBodyDefinitions [{
      boundBone {boneName=Leg_3_01_L}
      boxExtents=<25 25 25>
      constraintSetup {angularLimitsMax=<0 0 45>, angularLimitsMin=<0 0 -10>, angularTarget=<1 0 0>}
      localJointOffset=<-1 0 0>
     }]
    simSpaceSettings {maxAngularAcceleration=35, maxAngularVelocity=7, simSpaceAngularAlpha=1}
   }
   n20 = Anim Dynamics - Bone: Leg_2_01_L(ComponentPose=@n19, Alpha=0.2, ExternalForce="0,0,0") {
    angularDampingOverride=1
    angularSpringConstant=10000
    bAngularSpring=true
    bOverrideAngularDamping=true
    bUseGravityOverride=true
    gravityScale=0.35
    simulationSpace=RootRelative
    alphaBoolBlend {blendOption=Linear}
    boundBone {boneName=Leg_2_01_L}
    componentLinearAccScale=<1 1 1>
    componentLinearVelScale=<1 1 1>
    physicsBodyDefinitions [{
      boundBone {boneName=Leg_2_01_L}
      boxExtents=<25 25 25>
      constraintSetup {angularLimitsMax=<0 0 45>, angularLimitsMin=<0 0 -10>, angularTarget=<1 0 0>}
      localJointOffset=<-1 0 0>
     }]
    simSpaceSettings {maxAngularAcceleration=35, maxAngularVelocity=7, simSpaceAngularAlpha=1}
   }
   n21 = Anim Dynamics - Bone: Leg_1_01_L(ComponentPose=@n20, Alpha=0.2, ExternalForce="0,0,0") {
    angularDampingOverride=1
    angularSpringConstant=10000
    bAngularSpring=true
    bOverrideAngularDamping=true
    bUseGravityOverride=true
    gravityScale=0.35
    simulationSpace=RootRelative
    alphaBoolBlend {blendOption=Linear}
    boundBone {boneName=Leg_1_01_L}
    componentLinearAccScale=<1 1 1>
    componentLinearVelScale=<1 1 1>
    physicsBodyDefinitions [{
      boundBone {boneName=Leg_1_01_L}
      boxExtents=<25 25 25>
      constraintSetup {angularLimitsMax=<0 0 45>, angularLimitsMin=<0 0 -10>, angularTarget=<1 0 0>}
      localJointOffset=<-1 0 0>
     }]
    simSpaceSettings {maxAngularAcceleration=35, maxAngularVelocity=7, simSpaceAngularAlpha=1}
   }
   n22 = Anim Dynamics - Bone: Leg_4_01_L(ComponentPose=@n21, Alpha=0.2, ExternalForce="0,0,0") {
    angularDampingOverride=1
    angularSpringConstant=10000
    bAngularSpring=true
    bOverrideAngularDamping=true
    bUseGravityOverride=true
    gravityScale=0.35
    simulationSpace=RootRelative
    alphaBoolBlend {blendOption=Linear}
    boundBone {boneName=Leg_4_01_L}
    componentLinearAccScale=<1 1 1>
    componentLinearVelScale=<1 1 1>
    physicsBodyDefinitions [{
      boundBone {boneName=Leg_4_01_L}
      boxExtents=<25 25 25>
      constraintSetup {angularLimitsMax=<0 0 45>, angularLimitsMin=<0 0 -10>, angularTarget=<1 0 0>}
      localJointOffset=<-1 0 0>
     }]
    simSpaceSettings {maxAngularAcceleration=35, maxAngularVelocity=7, simSpaceAngularAlpha=1}
   }
   n23 = Anim Dynamics - Bone: Leg_4_01_R(ComponentPose=@n22, Alpha=0.2, ExternalForce="0,0,0") {
    angularDampingOverride=1
    angularSpringConstant=10000
    bAngularSpring=true
    bOverrideAngularDamping=true
    bUseGravityOverride=true
    gravityScale=0.35
    simulationSpace=RootRelative
    alphaBoolBlend {blendOption=Linear}
    boundBone {boneName=Leg_4_01_R}
    componentLinearAccScale=<1 1 1>
    componentLinearVelScale=<1 1 1>
    physicsBodyDefinitions [{
      boundBone {boneName=Leg_4_01_R}
      boxExtents=<25 25 25>
      constraintSetup {angularLimitsMax=<0 0 45>, angularLimitsMin=<0 0 -10>, angularTarget=<1 0 0>}
      localJointOffset=<1 0 0>
     }]
    simSpaceSettings {maxAngularAcceleration=35, maxAngularVelocity=7, simSpaceAngularAlpha=1}
   }
   n24 = Anim Dynamics - Bone: Leg_3_01_R(ComponentPose=@n23, Alpha=0.2, ExternalForce="0,0,0") {
    angularDampingOverride=1
    angularSpringConstant=10000
    bAngularSpring=true
    bOverrideAngularDamping=true
    bUseGravityOverride=true
    gravityScale=0.35
    simulationSpace=RootRelative
    alphaBoolBlend {blendOption=Linear}
    boundBone {boneName=Leg_3_01_R}
    componentLinearAccScale=<1 1 1>
    componentLinearVelScale=<1 1 1>
    physicsBodyDefinitions [{
      boundBone {boneName=Leg_3_01_R}
      boxExtents=<25 25 25>
      constraintSetup {angularLimitsMax=<0 0 45>, angularLimitsMin=<0 0 -10>, angularTarget=<1 0 0>}
      localJointOffset=<1 0 0>
     }]
    simSpaceSettings {maxAngularAcceleration=35, maxAngularVelocity=7, simSpaceAngularAlpha=1}
   }
   n25 = Anim Dynamics - Bone: Leg_2_01_R(ComponentPose=@n24, Alpha=0.2, ExternalForce="0,0,0") {
    angularDampingOverride=1
    angularSpringConstant=10000
    bAngularSpring=true
    bOverrideAngularDamping=true
    bUseGravityOverride=true
    gravityScale=0.35
    simulationSpace=RootRelative
    alphaBoolBlend {blendOption=Linear}
    boundBone {boneName=Leg_2_01_R}
    componentLinearAccScale=<1 1 1>
    componentLinearVelScale=<1 1 1>
    physicsBodyDefinitions [{
      boundBone {boneName=Leg_2_01_R}
      boxExtents=<25 25 25>
      constraintSetup {angularLimitsMax=<0 0 45>, angularLimitsMin=<0 0 -10>, angularTarget=<1 0 0>}
      localJointOffset=<1 0 0>
     }]
    simSpaceSettings {maxAngularAcceleration=35, maxAngularVelocity=7, simSpaceAngularAlpha=1}
   }
   n26 = Anim Dynamics - Bone: Leg_1_01_R(ComponentPose=@n25, Alpha=0.2, ExternalForce="0,0,0") {
    angularDampingOverride=1
    angularSpringConstant=10000
    bAngularSpring=true
    bOverrideAngularDamping=true
    bUseGravityOverride=true
    gravityScale=0.35
    simulationSpace=RootRelative
    alphaBoolBlend {blendOption=Linear}
    boundBone {boneName=Leg_1_01_R}
    componentLinearAccScale=<1 1 1>
    componentLinearVelScale=<1 1 1>
    physicsBodyDefinitions [{
      boundBone {boneName=Leg_1_01_R}
      boxExtents=<25 25 25>
      constraintSetup {angularLimitsMax=<0 0 45>, angularLimitsMin=<0 0 -10>, angularTarget=<1 0 0>}
      localJointOffset=<1 0 0>
     }]
    simSpaceSettings {maxAngularAcceleration=35, maxAngularVelocity=7, simSpaceAngularAlpha=1}
   }
   Output Pose(Result=Component To Local(ComponentPose=@n26)) {layerGroup=None}
  ]}
 wingPosition {
  entry=Base
  path=AnimGraph.AnimGraphNode_StateMachine_0.wingPosition
  ids=id
  nodes [
   AnimStateNode_0 = Base() {is=title AnimStateNode}
   AnimStateNode_1 = Jumping() {is=title AnimStateNode}
   AnimStateNode_2 = Falling() {is=title AnimStateNode}
   AnimStateNode_3 = Targeting() {is=title AnimStateNode}
   AnimStateNode_4 = Crouching() {is=title AnimStateNode}
   AnimStateNode_5 = SkydiveMode() {is=title AnimStateNode}
   AnimStateNode_9 = Tactical Sprint() {is=title AnimStateNode}
   AnimStateTransitionNode_0 = AnimStateTransitionNode() {blendMode=Custom, customBlendCurve=/Game/Accessories/FORT_Backpacks/M_MED_Demon_Pack/Animation/JumpTransitionBlendCurve, priorityOrder=2}
   AnimStateTransitionNode_1 = AnimStateTransitionNode() {
    blendMode=Custom
    crossfadeDuration=0.5
    customBlendCurve=/Game/Accessories/FORT_Backpacks/M_MED_Demon_Pack/Animation/JumpTransitionBlendOutCurve
   }
   AnimStateTransitionNode_10 = AnimStateTransitionNode()
   AnimStateTransitionNode_11 = AnimStateTransitionNode()
   AnimStateTransitionNode_12 = AnimStateTransitionNode()
   AnimStateTransitionNode_13 = AnimStateTransitionNode()
   AnimStateTransitionNode_14 = AnimStateTransitionNode()
   AnimStateTransitionNode_15 = AnimStateTransitionNode()
   AnimStateTransitionNode_2 = AnimStateTransitionNode()
   AnimStateTransitionNode_3 = AnimStateTransitionNode() {blendMode=Custom, customBlendCurve=&JumpTransitionBlendOutCurve}
   AnimStateTransitionNode_4 = AnimStateTransitionNode() {blendMode=Custom, customBlendCurve=&JumpTransitionBlendCurve}
   AnimStateTransitionNode_5 = AnimStateTransitionNode()
   AnimStateTransitionNode_6 = AnimStateTransitionNode() {crossfadeDuration=0.25}
   AnimStateTransitionNode_7 = AnimStateTransitionNode() {crossfadeDuration=0.25}
   AnimStateTransitionNode_8 = AnimStateTransitionNode()
   AnimStateTransitionNode_9 = AnimStateTransitionNode()
  ]
  links [
   {
    dst=Crouching
    rule=AnimGraph.AnimGraphNode_StateMachine_0.wingPosition.AnimStateTransitionNode_12.Transition
    src=Base
    via=AnimStateTransitionNode_12
   }
   {
    dst=Falling
    rule=AnimGraph.AnimGraphNode_StateMachine_0.wingPosition.AnimStateTransitionNode_4.Transition
    src=Base
    via=AnimStateTransitionNode_4
   }
   {
    dst=Jumping
    rule=AnimGraph.AnimGraphNode_StateMachine_0.wingPosition.AnimStateTransitionNode_0.Transition
    src=Base
    via=AnimStateTransitionNode_0
   }
   {
    dst=SkydiveMode
    rule=AnimGraph.AnimGraphNode_StateMachine_0.wingPosition.AnimStateTransitionNode_15.Transition
    src=Base
    via=AnimStateTransitionNode_15
   }
   {
    dst=Tactical Sprint
    rule=AnimGraph.AnimGraphNode_StateMachine_0.wingPosition.AnimStateTransitionNode_8.Transition
    src=Base
    via=AnimStateTransitionNode_8
   }
   {
    dst=Targeting
    rule=AnimGraph.AnimGraphNode_StateMachine_0.wingPosition.AnimStateTransitionNode_6.Transition
    src=Base
    via=AnimStateTransitionNode_6
   }
   {
    dst=Base
    rule=AnimGraph.AnimGraphNode_StateMachine_0.wingPosition.AnimStateTransitionNode_13.Transition
    src=Crouching
    via=AnimStateTransitionNode_13
   }
   {
    dst=Base
    rule=AnimGraph.AnimGraphNode_StateMachine_0.wingPosition.AnimStateTransitionNode_3.Transition
    src=Falling
    via=AnimStateTransitionNode_3
   }
   {
    dst=Base
    rule=AnimGraph.AnimGraphNode_StateMachine_0.wingPosition.AnimStateTransitionNode_1.Transition
    src=Jumping
    via=AnimStateTransitionNode_1
   }
   {
    dst=Falling
    rule=AnimGraph.AnimGraphNode_StateMachine_0.wingPosition.AnimStateTransitionNode_2.Transition
    src=Jumping
    via=AnimStateTransitionNode_2
   }
   {
    dst=SkydiveMode
    rule=AnimGraph.AnimGraphNode_StateMachine_0.wingPosition.AnimStateTransitionNode_5.Transition
    src=Jumping
    via=AnimStateTransitionNode_5
   }
   {
    dst=Tactical Sprint
    rule=AnimGraph.AnimGraphNode_StateMachine_0.wingPosition.AnimStateTransitionNode_14.Transition
    src=Jumping
    via=AnimStateTransitionNode_14
   }
   {
    dst=Base
    rule=AnimGraph.AnimGraphNode_StateMachine_0.wingPosition.AnimStateTransitionNode_11.Transition
    src=SkydiveMode
    via=AnimStateTransitionNode_11
   }
   {
    dst=Base
    rule=AnimGraph.AnimGraphNode_StateMachine_0.wingPosition.AnimStateTransitionNode_9.Transition
    src=Tactical Sprint
    via=AnimStateTransitionNode_9
   }
   {
    dst=Jumping
    rule=AnimGraph.AnimGraphNode_StateMachine_0.wingPosition.AnimStateTransitionNode_10.Transition
    src=Tactical Sprint
    via=AnimStateTransitionNode_10
   }
   {
    dst=Base
    rule=AnimGraph.AnimGraphNode_StateMachine_0.wingPosition.AnimStateTransitionNode_7.Transition
    src=Targeting
    via=AnimStateTransitionNode_7
   }
  ]
 }
 Base {
  path=AnimGraph.AnimGraphNode_StateMachine_0.wingPosition.AnimStateNode_0.Base
  nodes [
   n1 = Get PlayerVector_FwdBwd() -> (PlayerVector_FwdBwd:double)
   n2 = Get PlayerVector_LR() -> (PlayerVector_LR:double)
   n3 = TangyRadishMagna_Wing_Movement_Poses_PoseAsset : Base!(PoseWeight:float=1) -> (Pose:Pose Link Structure) {
    is=title AnimGraphNode_PoseByName
    poseAsset=/BRCosmetics/Accessories/FORT_Backpacks/Backpack_F_MED_TangyRadishMagma/Animation/TangyRadishMagna_Wing_Movement_Poses_PoseAsset
    poseName=Base
   }
   Output Animation Pose(Result=Apply Additive(
     Base=@n3
     Additive=Blendspace Player 'BS_TangyRadishMagma_Leans_Ground'(X=@n2, Y=@n1) {blendSpace=/BRCosmetics/Accessories/FORT_Backpacks/Backpack_F_MED_TangyRadishMagma/Animation/BS_TangyRadishMagma_Leans_Ground}
    ) {alphaBoolBlend {blendOption=Linear}})
  ]
 }
 Jumping {
  path=AnimGraph.AnimGraphNode_StateMachine_0.wingPosition.AnimStateNode_1.Jumping
  nodes [
   n1 = Get PlayerVector_FwdBwd()
   n2 = Get PlayerVector_LR()
   n3 = TangyRadishMagna_Wing_Movement_Poses_PoseAsset : Jump!(PoseWeight:float=1) -> (Pose:Pose Link Structure) {
    is=title AnimGraphNode_PoseByName
    poseAsset=&TangyRadishMagna_Wing_Movement_Poses_PoseAsset
    poseName=Jump
   }
   Output Animation Pose(Result=Apply Additive(
     Base=@n3
     Additive=Blendspace Player 'BS_TangyRadishMagma_Leans_Ground'(X=@n2, Y=@n1) {blendSpace=&BS_TangyRadishMagma_Leans_Ground}
    ) {alphaBoolBlend {blendOption=Linear}})
  ]
 }
 Falling {
  path=AnimGraph.AnimGraphNode_StateMachine_0.wingPosition.AnimStateNode_2.Falling
  nodes [
   n1 = Get PlayerVector_FwdBwd()
   n2 = Get PlayerVector_LR()
   n3 = TangyRadishMagna_Wing_Movement_Poses_PoseAsset : Jump(PoseWeight=1) {
    poseAsset=&TangyRadishMagna_Wing_Movement_Poses_PoseAsset
    poseName=Jump
   }
   Output Animation Pose(Result=Apply Additive(
     Base=@n3
     Additive=Blendspace Player 'BS_TangyRadishMagma_Leans_Ground'(X=@n2, Y=@n1) {blendSpace=&BS_TangyRadishMagma_Leans_Ground}
    ) {alphaBoolBlend {blendOption=Linear}})
  ]
 }
 Targeting {
  path=AnimGraph.AnimGraphNode_StateMachine_0.wingPosition.AnimStateNode_3.Targeting
  nodes [
   n1 = TangyRadishMagna_Wing_Movement_Poses_PoseAsset : Targeting!(PoseWeight:float=1) -> (Pose:Pose Link Structure) {
    is=title AnimGraphNode_PoseByName
    poseAsset=&TangyRadishMagna_Wing_Movement_Poses_PoseAsset
    poseName=Targeting
   }
   Output Animation Pose(Result=@n1)
  ]
 }
 Crouching {
  path=AnimGraph.AnimGraphNode_StateMachine_0.wingPosition.AnimStateNode_4.Crouching
  nodes [
   n1 = TangyRadishMagna_Wing_Movement_Poses_PoseAsset : Crouch!(PoseWeight:float=1) -> (Pose:Pose Link Structure) {
    is=title AnimGraphNode_PoseByName
    poseAsset=&TangyRadishMagna_Wing_Movement_Poses_PoseAsset
    poseName=Crouch
   }
   Output Animation Pose(Result=@n1)
  ]
 }
 SkydiveMode {
  path=AnimGraph.AnimGraphNode_StateMachine_0.wingPosition.AnimStateNode_5.SkydiveMode
  nodes [
   AnimGraphNode_StateMachine_0 = Skydiving State() {is=title AnimGraphNode_StateMachine}
   Output Animation Pose(Result=@AnimGraphNode_StateMachine_0)
  ]
 }
 Skydiving State {
  entry=Parachuting
  path=AnimGraph.AnimGraphNode_StateMachine_0.wingPosition.AnimStateNode_5.SkydiveMode.AnimGraphNode_StateMachine_0.Skydiving State
  ids=id
  nodes [
   AnimStateNode_1 = Diving() {is=title AnimStateNode}
   AnimStateNode_2 = Floating() {is=title AnimStateNode}
   AnimStateNode_4 = Parachuting() {is=title AnimStateNode}
   AnimStateTransitionNode_1 = AnimStateTransitionNode() {crossfadeDuration=0.5}
   AnimStateTransitionNode_10 = AnimStateTransitionNode()
   AnimStateTransitionNode_12 = AnimStateTransitionNode()
   AnimStateTransitionNode_13 = AnimStateTransitionNode()
   AnimStateTransitionNode_3 = AnimStateTransitionNode()
   AnimStateTransitionNode_5 = AnimStateTransitionNode() {crossfadeDuration=0.5}
  ]
  links [
   {
    dst=Floating
    rule=AnimGraph.AnimGraphNode_StateMachine_0.wingPosition.AnimStateNode_5.SkydiveMode.AnimGraphNode_StateMachine_0.Skydiving State.AnimStateTransitionNode_1.AnimationTransitionGraph_0
    src=Diving
    via=AnimStateTransitionNode_1
   }
   {
    dst=Parachuting
    rule=AnimGraph.AnimGraphNode_StateMachine_0.wingPosition.AnimStateNode_5.SkydiveMode.AnimGraphNode_StateMachine_0.Skydiving State.AnimStateTransitionNode_10.AnimationTransitionGraph_1
    src=Diving
    via=AnimStateTransitionNode_10
   }
   {
    dst=Diving
    rule=AnimGraph.AnimGraphNode_StateMachine_0.wingPosition.AnimStateNode_5.SkydiveMode.AnimGraphNode_StateMachine_0.Skydiving State.AnimStateTransitionNode_5.AnimationTransitionGraph_0
    src=Floating
    via=AnimStateTransitionNode_5
   }
   {
    dst=Parachuting
    rule=AnimGraph.AnimGraphNode_StateMachine_0.wingPosition.AnimStateNode_5.SkydiveMode.AnimGraphNode_StateMachine_0.Skydiving State.AnimStateTransitionNode_3.AnimationTransitionGraph_1
    src=Floating
    via=AnimStateTransitionNode_3
   }
   {
    dst=Diving
    rule=AnimGraph.AnimGraphNode_StateMachine_0.wingPosition.AnimStateNode_5.SkydiveMode.AnimGraphNode_StateMachine_0.Skydiving State.AnimStateTransitionNode_13.AnimationTransitionGraph_1
    src=Parachuting
    via=AnimStateTransitionNode_13
   }
   {
    dst=Floating
    rule=AnimGraph.AnimGraphNode_StateMachine_0.wingPosition.AnimStateNode_5.SkydiveMode.AnimGraphNode_StateMachine_0.Skydiving State.AnimStateTransitionNode_12.AnimationTransitionGraph_1
    src=Parachuting
    via=AnimStateTransitionNode_12
   }
  ]
 }
 Diving {
  path=AnimGraph.AnimGraphNode_StateMachine_0.wingPosition.AnimStateNode_5.SkydiveMode.AnimGraphNode_StateMachine_0.Skydiving State.AnimStateNode_1.Diving
  nodes [
   n1 = Get PlayerVector_FwdBwd()
   n2 = Get PlayerVector_LR()
   n3 = TangyRadishMagna_Wing_Movement_Poses_PoseAsset : Dive!(PoseWeight:float=1) -> (Pose:Pose Link Structure) {
    is=title AnimGraphNode_PoseByName
    poseAsset=&TangyRadishMagna_Wing_Movement_Poses_PoseAsset
    poseName=Dive
   }
   Output Animation Pose(Result=Apply Additive(
     Base=@n3
     Additive=Blendspace Player 'BS_TangyRadishMagma_Leans_Air'(X=@n2, Y=@n1) {blendSpace=/BRCosmetics/Accessories/FORT_Backpacks/Backpack_F_MED_TangyRadishMagma/Animation/BS_TangyRadishMagma_Leans_Air}
    ) {alphaBoolBlend {blendOption=Linear}})
  ]
 }
 Floating {
  path=AnimGraph.AnimGraphNode_StateMachine_0.wingPosition.AnimStateNode_5.SkydiveMode.AnimGraphNode_StateMachine_0.Skydiving State.AnimStateNode_2.Floating
  nodes [
   n1 = Get PlayerVector_FwdBwd()
   n2 = Get PlayerVector_LR()
   n3 = TangyRadishMagna_Wing_Movement_Poses_PoseAsset : Float!(PoseWeight:float=1) -> (Pose:Pose Link Structure) {
    is=title AnimGraphNode_PoseByName
    poseAsset=&TangyRadishMagna_Wing_Movement_Poses_PoseAsset
    poseName=Float
   }
   Output Animation Pose(Result=Apply Additive(
     Base=@n3
     Additive=Blendspace Player 'BS_TangyRadishMagma_Leans_Air'(X=@n2, Y=@n1) {blendSpace=&BS_TangyRadishMagma_Leans_Air}
    ) {alphaBoolBlend {blendOption=Linear}})
  ]
 }
 Parachuting {
  path=AnimGraph.AnimGraphNode_StateMachine_0.wingPosition.AnimStateNode_5.SkydiveMode.AnimGraphNode_StateMachine_0.Skydiving State.AnimStateNode_4.Parachuting
  nodes [
   n1 = Get PlayerVector_FwdBwd()
   n2 = Get PlayerVector_LR()
   n3 = TangyRadishMagna_Wing_Movement_Poses_PoseAsset : Float(PoseWeight=1) {
    poseAsset=&TangyRadishMagna_Wing_Movement_Poses_PoseAsset
    poseName=Float
   }
   Output Animation Pose(Result=Apply Additive(
     Base=@n3
     Additive=Blendspace Player 'BS_TangyRadishMagma_Leans_Air'(X=@n2, Y=@n1) {blendSpace=&BS_TangyRadishMagma_Leans_Air}
    ) {alphaBoolBlend {blendOption=Linear}})
  ]
 }
 AnimationTransitionGraph_0 {
  path=AnimGraph.AnimGraphNode_StateMachine_0.wingPosition.AnimStateNode_5.SkydiveMode.AnimGraphNode_StateMachine_0.Skydiving State.AnimStateTransitionNode_1.AnimationTransitionGraph_0
  nodes [Result(bCanEnterTransition=NOT Boolean(A=Get bIsSkydiveDiveMode()))]
 }
 AnimationTransitionGraph_1 {
  path=AnimGraph.AnimGraphNode_StateMachine_0.wingPosition.AnimStateNode_5.SkydiveMode.AnimGraphNode_StateMachine_0.Skydiving State.AnimStateTransitionNode_10.AnimationTransitionGraph_1
  nodes [Result(bCanEnterTransition=Get bIsParachuteOpened())]
 }
 AnimationTransitionGraph_1 {
  path=AnimGraph.AnimGraphNode_StateMachine_0.wingPosition.AnimStateNode_5.SkydiveMode.AnimGraphNode_StateMachine_0.Skydiving State.AnimStateTransitionNode_12.AnimationTransitionGraph_1
  nodes [Result(bCanEnterTransition=NOT Boolean(A=Get bIsParachuteOpened()))]
 }
 AnimationTransitionGraph_1 {
  path=AnimGraph.AnimGraphNode_StateMachine_0.wingPosition.AnimStateNode_5.SkydiveMode.AnimGraphNode_StateMachine_0.Skydiving State.AnimStateTransitionNode_13.AnimationTransitionGraph_1
  nodes [Result(bCanEnterTransition=NOT Boolean(A=Get bIsParachuteOpened()))]
 }
 AnimationTransitionGraph_1 {
  path=AnimGraph.AnimGraphNode_StateMachine_0.wingPosition.AnimStateNode_5.SkydiveMode.AnimGraphNode_StateMachine_0.Skydiving State.AnimStateTransitionNode_3.AnimationTransitionGraph_1
  nodes [Result(bCanEnterTransition=Get bIsParachuteOpened())]
 }
 AnimationTransitionGraph_0 {
  path=AnimGraph.AnimGraphNode_StateMachine_0.wingPosition.AnimStateNode_5.SkydiveMode.AnimGraphNode_StateMachine_0.Skydiving State.AnimStateTransitionNode_5.AnimationTransitionGraph_0
  nodes [Result(bCanEnterTransition=Get bIsSkydiveDiveMode())]
 }
 Tactical Sprint {
  path=AnimGraph.AnimGraphNode_StateMachine_0.wingPosition.AnimStateNode_9.Tactical Sprint
  nodes [
   n1 = TangyRadishMagna_Wing_Movement_Poses_PoseAsset : Sprint!(PoseWeight:float=1) -> (Pose:Pose Link Structure) {
    is=title AnimGraphNode_PoseByName
    poseAsset=&TangyRadishMagna_Wing_Movement_Poses_PoseAsset
    poseName=Sprint
   }
   Output Animation Pose(Result=@n1)
  ]
 }
 Transition {
  path=AnimGraph.AnimGraphNode_StateMachine_0.wingPosition.AnimStateTransitionNode_0.Transition
  nodes [Result(bCanEnterTransition=Get bIsJumping())]
 }
 Transition {
  path=AnimGraph.AnimGraphNode_StateMachine_0.wingPosition.AnimStateTransitionNode_1.Transition
  nodes [
   Result(bCanEnterTransition=NOT Boolean(A=Get bIsJumping()))
   Get bShouldPredictLanding() -> (bShouldPredictLanding:bool)
   Get bLandingPredicted() -> (bLandingPredicted:bool)
   Get bDebugLandPrediction() -> (bDebugLandPrediction:bool)
   Get bIsFalling()
  ]
 }
 Transition {
  path=AnimGraph.AnimGraphNode_StateMachine_0.wingPosition.AnimStateTransitionNode_10.Transition
  nodes [Result(bCanEnterTransition=Get bIsJumping())]
 }
 Transition {
  path=AnimGraph.AnimGraphNode_StateMachine_0.wingPosition.AnimStateTransitionNode_11.Transition
  nodes [Result(bCanEnterTransition=NOT Boolean(A=Get bIsSkyDiving()))]
 }
 Transition {
  path=AnimGraph.AnimGraphNode_StateMachine_0.wingPosition.AnimStateTransitionNode_12.Transition
  nodes [Result(bCanEnterTransition=Get bIsCrouching())]
 }
 Transition {
  path=AnimGraph.AnimGraphNode_StateMachine_0.wingPosition.AnimStateTransitionNode_13.Transition
  nodes [Result(bCanEnterTransition=NOT Boolean(A=Get bIsCrouching()))]
 }
 Transition {
  path=AnimGraph.AnimGraphNode_StateMachine_0.wingPosition.AnimStateTransitionNode_14.Transition
  nodes [Result(bCanEnterTransition=NOT Boolean(A=Get bIsJumping()))]
 }
 Transition {
  path=AnimGraph.AnimGraphNode_StateMachine_0.wingPosition.AnimStateTransitionNode_15.Transition
  nodes [Result(bCanEnterTransition=Get bIsSkyDiving())]
 }
 Transition {
  path=AnimGraph.AnimGraphNode_StateMachine_0.wingPosition.AnimStateTransitionNode_2.Transition
  nodes [Result(bCanEnterTransition=Get bIsFalling())]
 }
 Transition {
  path=AnimGraph.AnimGraphNode_StateMachine_0.wingPosition.AnimStateTransitionNode_3.Transition
  nodes [Result(bCanEnterTransition=NOT Boolean(A=Get bIsFalling()))]
 }
 Transition {
  path=AnimGraph.AnimGraphNode_StateMachine_0.wingPosition.AnimStateTransitionNode_4.Transition
  nodes [Result(bCanEnterTransition=Get bIsFalling())]
 }
 Transition {
  path=AnimGraph.AnimGraphNode_StateMachine_0.wingPosition.AnimStateTransitionNode_5.Transition
  nodes [Result(bCanEnterTransition=Get bIsSkyDiving())]
 }
 Transition {
  path=AnimGraph.AnimGraphNode_StateMachine_0.wingPosition.AnimStateTransitionNode_6.Transition
  nodes [Result(bCanEnterTransition=Get bIsTargeting())]
 }
 Transition {
  path=AnimGraph.AnimGraphNode_StateMachine_0.wingPosition.AnimStateTransitionNode_7.Transition
  nodes [Result(bCanEnterTransition=NOT Boolean(A=Get bIsTargeting()))]
 }
 Transition {
  path=AnimGraph.AnimGraphNode_StateMachine_0.wingPosition.AnimStateTransitionNode_8.Transition
  nodes [Result(bCanEnterTransition=Get bIsTacticalSprinting())]
 }
 Transition {
  path=AnimGraph.AnimGraphNode_StateMachine_0.wingPosition.AnimStateTransitionNode_9.Transition
  nodes [Result(bCanEnterTransition=NOT Boolean(A=Get bIsTacticalSprinting()))]
 }
 EventGraph {nodes [
   Event BlueprintUpdateAnimation()
   > n1 = PlayerMovement(AnimInstance=Self-Reference(), XDotProductCoefficient=1, YDotProductCoefficient=1, ZDotProductCoefficient=1)
   > Set PlayerVector_LR(PlayerVector_LR:double=@n1) -> (Output_Get:double)
   > Set PlayerVector_FwdBwd(PlayerVector_FwdBwd:double=@n1.YMovement) -> (Output_Get:double)
   Event OnBeginSkydiving()
   > Set DynamicsAlpha(DynamicsAlpha:double=1) -> (Output_Get:double)
   n2 = Event OnCharacterCustomizationLoaded()
   > Set Skel_Mesh(Skel_Mesh=@n2.CustomBaseMesh)
   > n3 = Branch(Condition=Get bIsInFrontEnd())
   > Set Petals Sim Space(Petals Sim Space:AnimPhysSimSpaceType Enum=World) -> (Output_Get:AnimPhysSimSpaceType Enum)
   > Set Gravity Override Petals Dynamics(Gravity Override Petals Dynamics:Vector="0,0,-10000") -> (Output_Get:Vector)
   > Set Petals Alpha FR(Petals Alpha FR:double=0.7) -> (Output_Get:double)
   Set Petals Sim Space(execute=@n3.else)
   > Set Gravity Override Petals Dynamics(Gravity Override Petals Dynamics="0,0,-980")
   > Set Petals Alpha FR(Petals Alpha FR=1)
   Event OnEndSkydiving()
   > Set DynamicsAlpha(DynamicsAlpha=1)
   TryGetPawnOwner()
  ]}
]
variables [
 DynamicsAlpha {type=number}
 Gravity Override Petals Dynamics {type=Vector, default=<0 0 -5000>}
 Petals Alpha FR {type=number}
 Petals Sim Space {type=AnimPhysSimSpaceType}
 PlayerVector_FwdBwd {type=number}
 PlayerVector_LR {type=number}
 Skel_Mesh {type=SkeletalMeshComponent}
]
